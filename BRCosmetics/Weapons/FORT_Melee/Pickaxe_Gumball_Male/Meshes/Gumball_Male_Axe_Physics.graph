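# Physics asset for the Gumball Male pickaxe mesh, a hammer carrying a small dangling gumball figure with simulated limbs.
# The head sphere is angularly locked to the hammer base capsule and anchors shoulder-hand and thigh-knee-foot capsule chains on both sides.
# Limb joints have narrow swing and twist limits that differ between left and right.
# A separate gumball base sphere hangs off the hammer, swinging mainly on one axis with some twist.
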PhysicsAsset /BRCosmetics/Weapons/FORT_Melee/Pickaxe_Gumball_Male/Meshes/Gumball_Male_Axe_Physics
bodies {
 dyn_foot_l [{shapeType=Capsule, length=1, radius=1, shapeName=dyn_foot_l_capsule, center=<1 0 0>, rotation=<rot 90 0 0>}]
 dyn_foot_r [{shapeType=Capsule, length=1, radius=1, shapeName=dyn_foot_r_capsule, center=<1 0 0>, rotation=<rot 90 0 0>}]
 dyn_gumball_head [{shapeType=Sphere, radius=7, shapeName=dyn_gumball_head_sphere}]
 dyn_hammer_gumball_base [{shapeType=Sphere, radius=8, shapeName=dyn_hammer_gumball_base_sphere, center=<6 0 0>}]
 dyn_hand_l [{shapeType=Capsule, length=2, radius=1, shapeName=dyn_hand_l_capsule, center=<1.5 0 0>, rotation=<rot 90 0 0>}]
 dyn_hand_r [{shapeType=Capsule, length=2, radius=1, shapeName=dyn_hand_r_capsule, center=<1.5 0 0>, rotation=<rot 90 0 0>}]
 dyn_knee_l [{shapeType=Capsule, length=1, radius=1, shapeName=dyn_knee_l_capsule, center=<1 0 0>, rotation=<rot 90 0 0>}]
 dyn_knee_r [{shapeType=Capsule, length=1, radius=1, shapeName=dyn_knee_r_capsule, center=<1 0 0>, rotation=<rot 90 0 0>}]
 dyn_shouldel_l [{shapeType=Capsule, length=2, radius=1, shapeName=dyn_shouldel_l_capsule, center=<1.5 0 0>, rotation=<rot 90 0 0>}]
 dyn_shouldel_r [{shapeType=Capsule, length=2, radius=1, shapeName=dyn_shouldel_r_capsule, center=<1.5 0 0>, rotation=<rot 90 0 0>}]
 dyn_thigh_l [{shapeType=Capsule, length=0.5, radius=1, shapeName=dyn_thigh_l_capsule, center=<1 0 0>, rotation=<rot 90 0 0>}]
 dyn_thigh_r [{shapeType=Capsule, length=0.5, radius=1, shapeName=dyn_thigh_r_capsule, center=<1 0 0>, rotation=<rot 90 0 0>}]
 hammer_base [{shapeType=Capsule, length=1, radius=13, shapeName=hammer_base_capsule, center=<7 0 0>, rotation=<rot 90 0 0>}]
}
constraints [
 {bone1Name=dyn_shouldel_l, bone2Name=dyn_gumball_head, swing1LimitDegrees=20, swing2LimitDegrees=20, twistLimitDegrees=5}
 {bone1Name=dyn_hand_l, bone2Name=dyn_shouldel_l, swing1LimitDegrees=20, swing2LimitDegrees=20, twistLimitDegrees=5}
 {bone1Name=dyn_thigh_l, bone2Name=dyn_gumball_head, swing1LimitDegrees=15, swing2LimitDegrees=15, twistLimitDegrees=5}
 {bone1Name=dyn_knee_l, bone2Name=dyn_thigh_l, swing1LimitDegrees=15, swing2LimitDegrees=15, twistLimitDegrees=5}
 {bone1Name=dyn_foot_l, bone2Name=dyn_knee_l, swing1LimitDegrees=30, swing2LimitDegrees=30, twistLimitDegrees=1}
 {bone1Name=dyn_shouldel_r, bone2Name=dyn_gumball_head, swing1LimitDegrees=15, swing2LimitDegrees=15, twistLimitDegrees=5}
 {bone1Name=dyn_hand_r, bone2Name=dyn_shouldel_r, swing1LimitDegrees=15, swing2LimitDegrees=15, twistLimitDegrees=5}
 {bone1Name=dyn_thigh_r, bone2Name=dyn_gumball_head, swing1LimitDegrees=20, swing2LimitDegrees=20, twistLimitDegrees=5}
 {bone1Name=dyn_knee_r, bone2Name=dyn_thigh_r, swing1LimitDegrees=20, swing2LimitDegrees=20, twistLimitDegrees=5}
 {bone1Name=dyn_foot_r, bone2Name=dyn_knee_r, swing1LimitDegrees=40, swing2LimitDegrees=40, twistLimitDegrees=1}
 {
  bone1Name=dyn_gumball_head
  bone2Name=hammer_base
  swing1LimitDegrees=45
  swing1Motion=Locked
  swing2LimitDegrees=45
  swing2Motion=Locked
  twistLimitDegrees=45
  twistMotion=Locked
 }
 {bone1Name=dyn_hammer_gumball_base, bone2Name=hammer_base, swing1LimitDegrees=1, swing2LimitDegrees=45, twistLimitDegrees=45}
]
